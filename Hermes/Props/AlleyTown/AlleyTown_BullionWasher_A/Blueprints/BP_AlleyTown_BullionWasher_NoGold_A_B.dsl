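; Destructible AlleyTown bullion washer prop, the "NoGold" variant, with an instanced washer mesh and side audio.
; The construction script sets a white secondary-mesh color and applies the configurable Washer Material to the instanced washers.
; Exposed asset provider overrides add a version and register that material as an allowed asset reference.
; On death it quickly fades out the Side B audio; BeginPlay, overlap and Tick events are empty.
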(event EventBeginPlay)

(event Collision|EventActorBeginOverlap
  (params (OtherActor :type Actor%20Object%20Reference)))

(event EventTick
  (params (DeltaSeconds :type Float%20%28single-precision%29)))

(event EventOnDeathPlayEffects
  (params (Damage :type Float%20%28single-precision%29) (DamageTags :type Gameplay%20Tag%20Container%20Structure%20%28by%20ref%29) (Momentum :type Vector :default "0, 0, 0") (HitInfo :type Hit%20Result%20Structure%20%28by%20ref%29) (InstigatedBy :type Fort%20Pawn%20Object%20Reference) (DamageCauser :type Actor%20Object%20Reference) (EffectContext :type Gameplay%20Effect%20Context%20Handle%20Structure))
  (Audio|Components|Audio|FadeOut (Variables|Default|GetAudio_Side_B) 0.25 0.0 "Linear" :!class /Script/Engine.AudioComponent))

(fn BuildExposedAssetProviderVersion(Implementation)
  (params (Version :type Fort%20Exposed%20Asset%20Provider%20Version%20Structure%20%28by%20ref%29))
  (|CallParentFunction Version)
  (ExposedAssetProvider|AddVersion Version 1)
  (return))

(fn CollectAssetReferencestoAllow(Implementation)
  (params (Collector :type Fort%20Exposed%20Asset%20Reference%20Collector%20Structure%20%28by%20ref%29))
  (|CallParentFunction Collector)
  (ExposedAssetProvider|CollectObjectPropertyAssetReferences Collector 0 (Variables|Default|GetWasher%20Material))
  (return))

(fn ConstructionScript
  (bind _instancedwashers (Variables|Default|GetInstancedWashers))
  (Default|SetColoronSecondaryMesh "(R=1.000000,G=1.000000,B=1.000000,A=1.000000)" _instancedwashers)
  (Rendering|Material|SetMaterial _instancedwashers 0 (Variables|Default|GetWasher%20Material)))
