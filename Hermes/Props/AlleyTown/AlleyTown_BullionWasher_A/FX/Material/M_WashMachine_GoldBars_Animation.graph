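# Default Lit material for the AlleyTown bullion washer's gold bars, set up for Niagara mesh particles to animate them spinning.
# Time scaled by RotationSpeed and wrapped with Frac drives RotateAboutAxis around Y at the object pivot, behind a UseWPO? static switch.
# FixRotateAboutAxisNormals takes the same angle and pivot to correct normals for the rotation.
# Tumbler textures supply color multiplied by an Emissive parameter, a normal map and a mask-driven lerp.
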Material /Hermes/Props/AlleyTown/AlleyTown_BullionWasher_A/FX/Material/M_WashMachine_GoldBars_Animation
graph {
 properties {bAutomaticallySetUsageInEditor=false, bUsedWithNiagaraMeshParticles=true, usedShadingModels=Default Lit}
 nodes [
  Constant()
  LinearInterpolate(
   A=Constant()
   B=Constant() {r=0.2}
   Alpha=TextureSample() {samplerType=SAMPLERTYPE_Masks, texture=/Hermes/Props/AlleyTown/AlleyTown_BullionWasher_A/Textures/T_AlleyTown_BullionTumbler_A_S}.G
  )
  n1 = Constant3Vector() {constant=<rgba 0 1 0 0>}
  n2 = Frac(None=Multiply(A=Time(), B=ScalarParameter() {defaultValue=1, parameterName=RotationSpeed}))
  n3 = MaterialFunctionCall() {materialFunction=&ObjectPivotPoint}
  MaterialFunctionCall(Rotation Axis=@n1, Rotation Angle=@n2, Movement Mask=Constant() {r=1}, Normals Pivot=@n3.Object Pivot Location) {materialFunction=&FixRotateAboutAxisNormals}
  MaterialFunctionCall!(Rotation Amount=@n2, PivotPoint, WorldPosition) {materialFunction=/Engine/Functions/Engine_MaterialFunctions02/WorldPositionOffset/RotateAboutWorldAxis_cheap}
  Multiply(
   A=TextureSample() {texture=/Hermes/Props/AlleyTown/AlleyTown_BullionWasher_A/Textures/T_AlleyTown_BullionTumbler_A_D}.RGB
   B=ScalarParameter() {parameterName=Emissive}
  )
  n4 = WorldPosition() {worldPositionShaderOffset=WPT_ExcludeAllShaderOffsets}
  StaticSwitchParameter(
   True=RotateAboutAxis(NormalizedRotationAxis=@n1, RotationAngle=@n2, PivotPoint=@n3.Object Pivot Location, Position=@n4.XYZ)
   False=Constant()
  ) {defaultValue=true, parameterName=UseWPO?}
  TextureSample() {samplerType=SAMPLERTYPE_Normal, texture=/Hermes/Props/AlleyTown/AlleyTown_BullionWasher_A/Textures/T_AlleyTown_BullionTumbler_A_N}
  Transform(None=@n4.XYZ) {transformSourceType=TRANSFORMSOURCE_World, transformType=TRANSFORM_Local}
 ]
}
